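# Translucent unlit two-sided effect material for the female Rockstar outfit's ring FX, usable on Niagara mesh particles.
# Switch_wpo blends world position offset from SplineThicken to a Sprite billboard with a local-space offset.
# The same parameter blends solid particle color into a masked texture channel tinted by particle color and alpha.
# Clamped sub-image ID parameters select the atlas cell, and MF_FX_EmissiveAdjustments processes the emissive color.
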Material /BRCosmetics/Characters/Player/Female/Medium/Bodies/F_MED_Rockstar/FX/M_Rockstar_Ring
graph {
 properties {
  bUsedWithNiagaraMeshParticles=true
  blendMode=BLEND_Translucent
  pixelDepthOffsetMode=PDOM_Legacy
  shadingModel=MSM_Unlit
  twoSided=true
  usedShadingModels=Unlit
 }
 nodes [
  n1 = ScalarParameter() {defaultValue=0.7}
  n2 = VectorParameter() {defaultValue=<rgba 0 0 0 1>}
  LinearInterpolate(
   A=MaterialFunctionCall(WidthBase=@n1, WidthTip=@n1) {materialFunction=&SplineThicken}.WorldPositionOffset
   B=Add(
    A=MaterialFunctionCall!(Center Location, 0-1 UVs, XY scale=@n1, OverridePivot, "Normals (Optional)") {materialFunction=&Sprite}.World position offset
    B=Transform(None=@n2.RGB) {transformSourceType=TRANSFORMSOURCE_Local}
   )
   Alpha=ScalarParameter() {parameterName=Switch_wpo}
  )
  n3 = TextureCoordinate()
  n4 = ScalarParameter() {defaultValue=2, group=SubUV_count, parameterName=SubImagesHorizontal}
  n5 = Subtract(A=ScalarParameter() {defaultValue=1, group=SubUV_ID, parameterName=SubImagesID_Horizontal})
  n6 = Subtract(A=@n4)
  n7 = Add(A=Divide(A=ComponentMask(None=@n3) {r=true}, B=@n4), B=Multiply(A=Divide(B=@n4), B=Clamp(None=@n5, Max=@n6)))
  n8 = ScalarParameter() {defaultValue=2, group=SubUV_count, parameterName=SubImagesVertical, sortPriority=1}
  n9 = Subtract(A=ScalarParameter() {defaultValue=2, group=SubUV_ID, parameterName=SubImagesID_Vertical, sortPriority=1})
  n10 = Subtract(A=@n8)
  n11 = Add(A=Divide(A=ComponentMask(None=@n3) {g=true}, B=@n8), B=Multiply(A=Divide(B=@n8), B=Clamp(None=@n9, Max=@n10)))
  n12 = StaticComponentMaskParameter(None=TextureSampleParameter2D(UVs=AppendVector(A=@n7, B=@n11)) {parameterName=SubImage_Texture, texture=/Game/Accessories/FORT_Backpacks/M_MED_Sleepy_Time_Backpack/FX/T_SleepyTime_Elements}.RGB) {defaultR=true, parameterName=Channel}
  n13 = ParticleColor()
  n14 = ScalarParameter() {parameterName=Switch_wpo}
  LinearInterpolate(A=Constant() {r=1}, B=Multiply(A=@n12, B=@n13.A), Alpha=@n14)
  MaterialFunctionCall(Emissive=LinearInterpolate(A=@n13.RGB, B=Multiply(A=@n12, B=@n13.RGB), Alpha=@n14)) {materialFunction=&MF_FX_EmissiveAdjustments}
 ]
}
