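# Physics asset for the right-arm sleeve of the F_MED_Rockstar female player cosmetic body.
# Four small capsules on the Dyn__Sleeve bones each constrain directly to upperarm_twist_01_r rather than chaining, with narrow swing and near-locked twist limits.
# Larger capsules on the upper-arm twist bone and spine_03 stand in for arm and torso, joined by a constraint with free angular motion.
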PhysicsAsset /BRCosmetics/Characters/Player/Female/Medium/Bodies/F_MED_Rockstar/Meshes/F_MED_Rockstar_ArmPhysics
bodies {
 Dyn__Sleeve1 [{shapeType=Capsule, length=3, radius=1, center=<1 0.2771 0>, rotation=<rot 90 0 0>}]
 Dyn__Sleeve2 [{shapeType=Capsule, length=3, radius=1, center=<1 0.1668 -1.7032>, rotation=<rot 90 0 0>}]
 Dyn__Sleeve3 [{shapeType=Capsule, length=2.3402, radius=0.7033, center=<1 -0.0978 0>, rotation=<rot 90 0 0>}]
 Dyn__Sleeve4 [{shapeType=Capsule, length=3, radius=1, center=<1 -0.0457 0>, rotation=<rot 90 180 180>}]
 spine_03 [{shapeType=Capsule, length=2, radius=7.5, center=<-4.543 1.8792 0.0001>}]
 upperarm_twist_01_r [{shapeType=Capsule, length=10.3483, radius=4.1517, center=<-2.4857 0.4681 1.005>, rotation=<rot 90 0 0>}]
}
constraints [
 {
  bone1Name=upperarm_twist_01_r
  bone2Name=spine_03
  swing1LimitDegrees=45
  swing1Motion=Free
  swing2LimitDegrees=45
  swing2Motion=Free
  twistLimitDegrees=45
  twistMotion=Free
 }
 {bone1Name=Dyn__Sleeve1, bone2Name=upperarm_twist_01_r, swing1LimitDegrees=20, swing2LimitDegrees=20, twistLimitDegrees=1}
 {bone1Name=Dyn__Sleeve2, bone2Name=upperarm_twist_01_r, swing1LimitDegrees=20, swing2LimitDegrees=20, twistLimitDegrees=1}
 {bone1Name=Dyn__Sleeve3, bone2Name=upperarm_twist_01_r, swing1LimitDegrees=20, swing2LimitDegrees=20, twistLimitDegrees=1}
 {bone1Name=Dyn__Sleeve4, bone2Name=upperarm_twist_01_r, swing1LimitDegrees=20, swing2LimitDegrees=20, twistLimitDegrees=1}
]
