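# Physics asset for the F_MED_Rockstar female player cosmetic body, providing capsule bodies for secondary motion on shirt, right sleeve and shoelaces.
# Eight shirt bones hang from spine_03 with tight swing and twist limits, as do four sleeve bones from upperarm_twist_01_r, itself angularly free against the spine.
# Each shoelace swings more widely from a master bone that is angularly locked to its calf twist bone.
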PhysicsAsset /BRCosmetics/Characters/Player/Female/Medium/Bodies/F_MED_Rockstar/Meshes/F_MED_Rockstar_Physics
bodies {
 Dyn_Shirt_Back2 [{shapeType=Capsule, length=5, radius=1, center=<1 -0.3424 0>, rotation=<rot 90 0 0>}]
 Dyn_Shirt_Back_l2 [{shapeType=Capsule, length=5, radius=1, center=<1 -0.2611 0>, rotation=<rot 90 0 0>}]
 Dyn_Shirt_Back_r2 [{shapeType=Capsule, length=5, radius=1, center=<1 -0.2611 0>, rotation=<rot 90 0 0>}]
 Dyn_Shirt_Front2 [{shapeType=Capsule, length=5, radius=1, center=<1 -0.5 0>, rotation=<rot 90 0 0>}]
 Dyn_Shirt_Front_l2 [{shapeType=Capsule, length=5, radius=1, center=<1 -0.6719 0>, rotation=<rot 90 0 0>}]
 Dyn_Shirt_Front_r2 [{shapeType=Capsule, length=5, radius=1, center=<1.0001 -0.9799 0>, rotation=<rot 90 0 0>}]
 Dyn_Shirt_Side_l2 [{shapeType=Capsule, length=5, radius=1, center=<1 -0.5 0>, rotation=<rot 90 0 0>}]
 Dyn_Shirt_Side_r2 [{shapeType=Capsule, length=5, radius=1, center=<1 -0.5 0>, rotation=<rot 90 0 0>}]
 Dyn_Shoelace_MSTR_l [{shapeType=Capsule, length=0.505, radius=0.505, shapeName=Dyn_Shoelace_MSTR_l_capsule, rotation=<rot 0 -90 0>}]
 Dyn_Shoelace_MSTR_r [{shapeType=Capsule, length=0.505, radius=0.505, shapeName=Dyn_Shoelace_MSTR_r_capsule, rotation=<rot 0 -95 0>}]
 Dyn_Shoelace_l [{shapeType=Capsule, length=3, radius=1, center=<1 0 0>, rotation=<rot 90 0 0>}]
 Dyn_Shoelace_r [{shapeType=Capsule, length=3, radius=1, center=<1 0 0>, rotation=<rot 90 0 0>}]
 Dyn__Sleeve1 [{shapeType=Capsule, length=3, radius=1, center=<1 0 0>, rotation=<rot 90 0 0>}]
 Dyn__Sleeve2 [{shapeType=Capsule, length=3, radius=1, center=<1 0 0>, rotation=<rot 90 0 0>}]
 Dyn__Sleeve3 [{shapeType=Capsule, length=3, radius=1, center=<1 0 0>, rotation=<rot 90 0 0>}]
 Dyn__Sleeve4 [{shapeType=Capsule, length=3, radius=1, center=<1 0 0>, rotation=<rot 90 0 0>}]
 calf_twist_01_l [{shapeType=Capsule, length=0.5, radius=0.5}]
 calf_twist_01_r [{shapeType=Capsule, length=0.5, radius=0.5}]
 spine_03 [{shapeType=Capsule, length=2, radius=7.5, center=<-4.543 1.8792 0.0001>}]
 upperarm_twist_01_r [{shapeType=Capsule, length=10, radius=3.5, center=<0 0.4681 1.005>, rotation=<rot 90 0 0>}]
}
constraints [
 {bone1Name=Dyn_Shirt_Front2, bone2Name=spine_03, swing1LimitDegrees=10, swing2LimitDegrees=20, twistLimitDegrees=1}
 {bone1Name=Dyn_Shirt_Front_l2, bone2Name=spine_03, swing1LimitDegrees=10, swing2LimitDegrees=20, twistLimitDegrees=1}
 {bone1Name=Dyn_Shirt_Front_r2, bone2Name=spine_03, swing1LimitDegrees=10, swing2LimitDegrees=20, twistLimitDegrees=1}
 {bone1Name=Dyn_Shirt_Side_r2, bone2Name=spine_03, swing1LimitDegrees=10, swing2LimitDegrees=20, twistLimitDegrees=1}
 {bone1Name=Dyn_Shirt_Side_l2, bone2Name=spine_03, swing1LimitDegrees=10, swing2LimitDegrees=20, twistLimitDegrees=1}
 {bone1Name=Dyn_Shirt_Back_l2, bone2Name=spine_03, swing1LimitDegrees=10, swing2LimitDegrees=20, twistLimitDegrees=1}
 {bone1Name=Dyn_Shirt_Back_r2, bone2Name=spine_03, swing1LimitDegrees=10, swing2LimitDegrees=20, twistLimitDegrees=1}
 {bone1Name=Dyn_Shirt_Back2, bone2Name=spine_03, swing1LimitDegrees=10, swing2LimitDegrees=20, twistLimitDegrees=1}
 {bone1Name=Dyn_Shoelace_l, bone2Name=Dyn_Shoelace_MSTR_l, swing1LimitDegrees=45, swing2LimitDegrees=45, twistLimitDegrees=3}
 {bone1Name=Dyn_Shoelace_r, bone2Name=Dyn_Shoelace_MSTR_r, swing1LimitDegrees=45, swing2LimitDegrees=45, twistLimitDegrees=3}
 {
  bone1Name=upperarm_twist_01_r
  bone2Name=spine_03
  swing1LimitDegrees=45
  swing1Motion=Free
  swing2LimitDegrees=45
  swing2Motion=Free
  twistLimitDegrees=45
  twistMotion=Free
 }
 {bone1Name=Dyn__Sleeve1, bone2Name=upperarm_twist_01_r, swing1LimitDegrees=20, swing2LimitDegrees=20, twistLimitDegrees=1}
 {bone1Name=Dyn__Sleeve2, bone2Name=upperarm_twist_01_r, swing1LimitDegrees=20, swing2LimitDegrees=20, twistLimitDegrees=1}
 {bone1Name=Dyn__Sleeve3, bone2Name=upperarm_twist_01_r, swing1LimitDegrees=20, swing2LimitDegrees=20, twistLimitDegrees=1}
 {bone1Name=Dyn__Sleeve4, bone2Name=upperarm_twist_01_r, swing1LimitDegrees=20, swing2LimitDegrees=20, twistLimitDegrees=1}
 {
  bone1Name=Dyn_Shoelace_MSTR_r
  bone2Name=calf_twist_01_r
  swing1LimitDegrees=45
  swing1Motion=Locked
  swing2LimitDegrees=45
  swing2Motion=Locked
  twistLimitDegrees=45
  twistMotion=Locked
 }
 {
  bone1Name=Dyn_Shoelace_MSTR_l
  bone2Name=calf_twist_01_l
  swing1LimitDegrees=45
  swing1Motion=Locked
  swing2LimitDegrees=45
  swing2Motion=Locked
  twistLimitDegrees=45
  twistMotion=Locked
 }
]
