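# Animation Blueprint for the female medium Rockstar cosmetic body, adding secondary physics motion to the pose copied from its attached parent mesh.
# Two chained RigidBody nodes simulate body and arm physics assets relative to the pelvis and right clavicle, then AnimDynamics swings both shoelaces.
# Each lace has root-relative and world-space solvers, crossfaded by interpolated alphas.
# World space applies only in the front end outside emotes and skydive dive mode.
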AnimBlueprint /BRCosmetics/Characters/Player/Female/Medium/Bodies/F_MED_Rockstar/Meshes/F_MED_Rockstar_AnimBP
target_skeleton=&/F_MED_Rockstar_Skeleton
graphs [
 AnimGraph {nodes [
   n1 = RigidBody(ComponentPose=Local To Component(LocalPose=Copy Pose From Mesh() {bUseAttachedParent=true})) {
    bClampLinearTranslationLimitToRefPose=true
    bOverrideWorldGravity=true
    overridePhysicsAsset=&/F_MED_Rockstar_Physics
    simulationSpace=BaseBoneSpace
    alphaBoolBlend {blendOption=Linear}
    baseBoneRef {boneName=pelvis}
    componentAppliedLinearAccClamp=<1200 1200 1200>
    componentLinearAccScale=<0.1 0.5 0.1>
    componentLinearVelScale=<0.1 0.5>
    overrideWorldGravity {z=-980}
   }
   n2 = RigidBody(ComponentPose=@n1) {
    bClampLinearTranslationLimitToRefPose=true
    bOverrideWorldGravity=true
    overridePhysicsAsset=&/F_MED_Rockstar_ArmPhysics
    simulationSpace=BaseBoneSpace
    alphaBoolBlend {blendOption=Linear}
    baseBoneRef {boneName=clavicle_r}
    componentAppliedLinearAccClamp=<1200 1200 1200>
    componentLinearAccScale=<0.1 0.5 0.1>
    componentLinearVelScale=<0.1 0.5>
    overrideWorldGravity {z=-980}
   }
   n3 = Get Anim_Dyn_SimSpace_Root() -> (Anim_Dyn_SimSpace_Root:double)
   n4 = Anim Dynamics - Bone: Dyn_Shoelace_l(ComponentPose=@n2, Alpha=@n3, ExternalForce="0,0,0") {
    angularBiasOverride=0.99
    angularDampingOverride=0.99
    bOverrideAngularBias=true
    bOverrideAngularDamping=true
    bOverrideLinearDamping=true
    bUseGravityOverride=true
    linearDampingOverride=0.99
    numSolverIterationsPostUpdate=2
    numSolverIterationsPreUpdate=8
    simulationSpace=RootRelative
    alphaBoolBlend {blendOption=Linear}
    boundBone {boneName=Dyn_Shoelace_l}
    componentAppliedLinearAccClamp=<100 100 100>
    componentLinearAccScale=<0.1 0.1 0.1>
    componentLinearVelScale=<0.1 0.1 0.1>
    gravityOverride {z=-980}
    physicsBodyDefinitions [{
      boundBone {boneName=Dyn_Shoelace_l}
      boxExtents=<12 3 7>
      constraintSetup {angularLimitsMax=<4 40 0>, angularLimitsMin=<-4 -40 -130>, angularTarget=<0 0 0>}
      localJointOffset=<1.7 0 0>
     }]
    relativeSpaceBone {boneName=pelvis}
   }
   n5 = Anim Dynamics - Bone: Dyn_Shoelace_r(ComponentPose=@n4, Alpha=@n3, ExternalForce="0,0,0") {
    angularBiasOverride=0.99
    angularDampingOverride=0.99
    bOverrideAngularBias=true
    bOverrideAngularDamping=true
    bOverrideLinearDamping=true
    bUseGravityOverride=true
    linearDampingOverride=0.99
    numSolverIterationsPostUpdate=2
    numSolverIterationsPreUpdate=8
    simulationSpace=RootRelative
    alphaBoolBlend {blendOption=Linear}
    boundBone {boneName=Dyn_Shoelace_r}
    componentAppliedLinearAccClamp=<100 100 100>
    componentLinearAccScale=<0.1 0.1 0.1>
    componentLinearVelScale=<0.1 0.1 0.1>
    gravityOverride {z=-980}
    physicsBodyDefinitions [{
      boundBone {boneName=Dyn_Shoelace_r}
      boxExtents=<12 3 7>
      constraintSetup {angularLimitsMax=<4 40 0>, angularLimitsMin=<-4 -40 -130>, angularTarget=<0 0 0>}
      localJointOffset=<1.7 0 0>
     }]
    relativeSpaceBone {boneName=pelvis}
   }
   n6 = Get Anim_Dyn_SimSpace_World() -> (Anim_Dyn_SimSpace_World:double)
   n7 = Anim Dynamics - Bone: Dyn_Shoelace_l(ComponentPose=@n5, Alpha=@n6, ExternalForce="0,0,0") {
    angularBiasOverride=0.99
    angularDampingOverride=0.99
    bOverrideAngularBias=true
    bOverrideAngularDamping=true
    bOverrideLinearDamping=true
    bUseGravityOverride=true
    linearDampingOverride=0.99
    numSolverIterationsPostUpdate=2
    numSolverIterationsPreUpdate=8
    simulationSpace=World
    alphaBoolBlend {blendOption=Linear}
    boundBone {boneName=Dyn_Shoelace_l}
    componentAppliedLinearAccClamp=<100 100 100>
    componentLinearAccScale=<0.1 0.1 0.1>
    componentLinearVelScale=<0.1 0.1 0.1>
    gravityOverride {z=-980}
    physicsBodyDefinitions [{
      boundBone {boneName=Dyn_Shoelace_l}
      boxExtents=<12 3 7>
      constraintSetup {angularLimitsMax=<4 40 0>, angularLimitsMin=<-4 -40 -130>, angularTarget=<0 0 0>}
      localJointOffset=<1.7 0 0>
     }]
    relativeSpaceBone {boneName=pelvis}
   }
   n8 = Anim Dynamics - Bone: Dyn_Shoelace_r(ComponentPose=@n7, Alpha=@n6, ExternalForce="0,0,0") {
    angularBiasOverride=0.99
    angularDampingOverride=0.99
    bOverrideAngularBias=true
    bOverrideAngularDamping=true
    bOverrideLinearDamping=true
    bUseGravityOverride=true
    linearDampingOverride=0.99
    numSolverIterationsPostUpdate=2
    numSolverIterationsPreUpdate=8
    simulationSpace=World
    alphaBoolBlend {blendOption=Linear}
    boundBone {boneName=Dyn_Shoelace_r}
    componentAppliedLinearAccClamp=<100 100 100>
    componentLinearAccScale=<0.1 0.1 0.1>
    componentLinearVelScale=<0.1 0.1 0.1>
    gravityOverride {z=-980}
    physicsBodyDefinitions [{
      boundBone {boneName=Dyn_Shoelace_r}
      boxExtents=<12 3 7>
      constraintSetup {angularLimitsMax=<4 40 0>, angularLimitsMin=<-4 -40 -130>, angularTarget=<0 0 0>}
      localJointOffset=<1.7 0 0>
     }]
    relativeSpaceBone {boneName=pelvis}
   }
   Output Pose(Result=Component To Local(ComponentPose=@n8)) {layerGroup=None}
  ]}
 EventGraph {nodes [
   n1 = Event BlueprintUpdateAnimation()
   n2 = Get bIsPlayingEmote()
   n3 = Get bIsSkydiveDiveMode()
   > n4 = Branch(Condition=AND Boolean!(A=Get bIsInFrontEnd(), B=NOT Boolean(A=@n3), C:bool=NOT Boolean(A=@n2)) -> (ReturnValue))
   > Set Anim_Dyn_SimSpace_World(Anim_Dyn_SimSpace_World:double=FInterpTo_Constant!(Current:double=Get Anim_Dyn_SimSpace_World(), Target:double=1, DeltaTime:double=@n1.DeltaTimeX, InterpSpeed:double=2) -> (ReturnValue:double)) -> (Output_Get:double)
   > Set Anim_Dyn_SimSpace_Root(Anim_Dyn_SimSpace_Root:double=FInterpTo_Constant(Current=Get Anim_Dyn_SimSpace_Root(), Target=0, DeltaTime=@n1.DeltaTimeX, InterpSpeed=2)) -> (Output_Get:double)
   Set Anim_Dyn_SimSpace_World(
    execute=@n4.else
    Anim_Dyn_SimSpace_World=FInterpTo_Constant(Current=Get Anim_Dyn_SimSpace_World(), Target=0, DeltaTime=@n1.DeltaTimeX, InterpSpeed=2)
   )
   > Set Anim_Dyn_SimSpace_Root(Anim_Dyn_SimSpace_Root=FInterpTo_Constant(Current=Get Anim_Dyn_SimSpace_Root(), Target=1, DeltaTime=@n1.DeltaTimeX, InterpSpeed=2))
  ]}
]
variables [
 Anim_Dyn_SimSpace_Root {type=number}
 Anim_Dyn_SimSpace_World {type=number}
]
