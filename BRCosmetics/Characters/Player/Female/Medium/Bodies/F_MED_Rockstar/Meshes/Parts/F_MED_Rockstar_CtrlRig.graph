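# Control Rig for the female Rockstar body part that corrects the back-hair base bones as the head and right arm move.
# Spherical pose readers on the head and right upper arm, relative to spine_05, weight additive transform offsets on those bones.
# A crouching branch swaps in a separate head-driven correction.
# When not crouching, extra right-side head and arm corrections are scaled by the rifle targeting alpha.
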ControlRigBlueprint /BRCosmetics/Characters/Player/Female/Medium/Bodies/F_MED_Rockstar/Meshes/Parts/F_MED_Rockstar_CtrlRig
graphs [
 RigVMFunctionLibrary {links [], nodes []}
 RigVMModel {nodes [
   RigUnit_BeginExecution = &RigUnit_BeginExecution()
   SequenceAggregate = &RigVMFunction_Sequence(ExecuteContext=@RigUnit_BeginExecution.ExecutePin)
   n1 = &RigVMFunction_ControlFlowBranch(ExecuteContext=@SequenceAggregate.A, Condition=RigVMVariableNode() {variable=crouching}.Value)
   n2 = &RigUnit_SphericalPoseReader(
    ActiveRegionScaleFactors=(PositiveWidth=2.5,NegativeWidth=3,PositiveHeight=3,NegativeHeight=3)
    ActiveRegionSize=0.1
    Debug=(bDrawDebug=false,DebugScale=25,DebugSegments=20,DebugThickness=0.25)
    DriverAxis=(X=1,Y=0,Z=0)
    DriverItem=(Type=Bone,Name="head")
    FalloffRegionScaleFactors=(PositiveWidth=1,NegativeWidth=1,PositiveHeight=1,NegativeHeight=1)
    FalloffSize=0.5
    FlipHeightScaling=true
    OptionalParentItem=(Type=Bone,Name="spine_05")
    RotationOffset=(X=-45,Y=0,Z=0)
    ExecutePin=@n1.False
   )
   n3 = &RigVMFunction_MathQuaternionFromEuler(Euler=(X=0,Y=0,Z=-35))
   > &RigUnit_ModifyTransforms(
    ItemToModify="((Item=(Type=Bone,Name=\"dyn_hairBack_01_Base_r\"),Transform=(Rotation=(X=0,Y=0,Z=0,W=1),Translation=(X=0,Y=0,Z=0),Scale3D=(X=1,Y=1,Z=1))),(Item=(Type=Bone,Name=\"dyn_hairBack_01_Base_l\"),Transform=(Rotation=(X=0,Y=0,Z=0,W=1),Translation=(X=0,Y=0,Z=0),Scale3D ...(truncated)"
    ItemToModify.0.Transform.Rotation=@n3.Result
    ItemToModify.1.Transform.Rotation=@n3.Result
    ItemToModify.2.Transform.Rotation=@n3.Result
    Weight=@n2.OutputParam
   )
   > &RigUnit_ModifyTransforms(
    ItemToModify=((Item=(Type=Bone,Name="dyn_hairBack_01_Base_r"),Transform=(Rotation=(X=0,Y=0,Z=0,W=1),Translation=(X=0,Y=0,Z=0),Scale3D=(X=1,Y=1,Z=1))))
    ItemToModify.0.Transform.Rotation=&RigVMFunction_MathQuaternionFromEuler(Euler=(X=10,Y=20,Z=-10)).Result
    Weight=@n2.OutputParam
   )
   > &RigUnit_ModifyTransforms(
    ItemToModify=((Item=(Type=Bone,Name="dyn_hairBack_01_c_BASE"),Transform=(Rotation=(X=0,Y=0,Z=0,W=1),Translation=(X=0,Y=0,Z=0),Scale3D=(X=1,Y=1,Z=1))))
    ItemToModify.0.Transform.Rotation=&RigVMFunction_MathQuaternionFromEuler(Euler=(X=0,Y=20,Z=-10)).Result
    Weight=@n2.OutputParam
   )
   > n4 = &RigUnit_SphericalPoseReader(
    ActiveRegionScaleFactors=(PositiveWidth=0.25,NegativeWidth=0.25,PositiveHeight=1,NegativeHeight=1)
    ActiveRegionSize=0.1
    Debug=(bDrawDebug=false,DebugScale=25,DebugSegments=20,DebugThickness=0.25)
    DriverAxis=(X=0,Y=1,Z=0)
    DriverItem=(Type=Bone,Name="head")
    FalloffRegionScaleFactors=(PositiveWidth=0.65,NegativeWidth=0.5,PositiveHeight=4,NegativeHeight=4)
    FalloffSize=0.2
    OptionalParentItem=(Type=Bone,Name="spine_05")
    RotationOffset=(X=90,Y=0,Z=25)
   )
   > &RigUnit_ModifyTransforms(
    ItemToModify=((Item=(Type=Bone,Name="dyn_hairBack_01_Base_r"),Transform=(Rotation=(X=0,Y=0,Z=0,W=1),Translation=(X=0,Y=0,Z=0),Scale3D=(X=1,Y=1,Z=1))))
    ItemToModify.0.Transform.Rotation=&RigVMFunction_MathQuaternionFromEuler(Euler=(X=-10,Y=-20,Z=-40)).Result
    Weight=&RigVMFunction_MathDoubleMul(A=@n4.OutputParam, B=RigVMVariableNode() {variable=Is Rifle Targeting Alpha}.Value).Result
   )
   > n5 = &RigUnit_SphericalPoseReader(
    ActiveRegionScaleFactors=(PositiveWidth=1.5,NegativeWidth=1.5,PositiveHeight=1.5,NegativeHeight=1.5)
    ActiveRegionSize=0.1
    Debug=(bDrawDebug=false,bDraw2D=false,bDrawLocalAxes=false,DebugScale=25,DebugSegments=20,DebugThickness=0.25)
    DriverAxis=(X=-1,Y=0,Z=0)
    DriverItem=(Type=Bone,Name="upperarm_r")
    FalloffRegionScaleFactors=(PositiveWidth=0.5,NegativeWidth=0.5,PositiveHeight=0.5,NegativeHeight=0.5)
    FalloffSize=0.285
    OptionalParentItem=(Type=Bone,Name="spine_05")
    RotationOffset=(X=-45,Y=0,Z=-30)
   )
   > &RigUnit_ModifyTransforms(
    ItemToModify=((Item=(Type=Bone,Name="dyn_hairBack_01_Base_r"),Transform=(Rotation=(X=0,Y=0,Z=0,W=1),Translation=(X=0,Y=0,Z=0),Scale3D=(X=1,Y=1,Z=1))))
    ItemToModify.0.Transform.Rotation=&RigVMFunction_MathQuaternionFromEuler(Euler=(X=-10,Y=-20,Z=-40)).Result
    Weight=&RigVMFunction_MathDoubleMul(A=@n5.OutputParam, B=RigVMVariableNode() {variable=Is Rifle Targeting Alpha}.Value).Result
   )
   n6 = &RigUnit_SphericalPoseReader(
    ActiveRegionScaleFactors=(PositiveWidth=2.5,NegativeWidth=3,PositiveHeight=3,NegativeHeight=3)
    ActiveRegionSize=0.1
    Debug=(bDrawDebug=false,DebugScale=25,DebugSegments=20,DebugThickness=0.25)
    DriverAxis=(X=1,Y=0,Z=0)
    DriverItem=(Type=Bone,Name="head")
    FalloffRegionScaleFactors=(PositiveWidth=1,NegativeWidth=1,PositiveHeight=1,NegativeHeight=1)
    FalloffSize=0.5
    FlipHeightScaling=true
    OptionalParentItem=(Type=Bone,Name="spine_05")
    RotationOffset=(X=-45,Y=0,Z=-45)
    ExecutePin=@n1.True
   )
   n7 = &RigVMFunction_MathQuaternionFromEuler(Euler=(X=0,Y=0,Z=-35))
   > &RigUnit_ModifyTransforms(
    ItemToModify="((Item=(Type=Bone,Name=\"dyn_hairBack_01_Base_r\"),Transform=(Rotation=(X=0,Y=0,Z=0,W=1),Translation=(X=-5,Y=-5,Z=0),Scale3D=(X=1,Y=1,Z=1))),(Item=(Type=Bone,Name=\"dyn_hairBack_01_Base_l\"),Transform=(Rotation=(X=0,Y=0,Z=0,W=1),Translation=(X=0,Y=0,Z=0),Scale ...(truncated)"
    ItemToModify.0.Transform.Rotation=@n7.Result
    ItemToModify.2.Transform.Rotation=@n7.Result
    ItemToModify.1.Transform.Rotation=&RigVMFunction_MathQuaternionFromEuler(Euler=(X=0,Y=0,Z=-35)).Result
    Weight=@n6.OutputParam
   )
  ]}
]
hierarchy [
 root {loc=<0 0 0>, rot=<0 0 0 1>}
 pelvis {parent=root, loc=<0 -0.5301 98.0571>, rot=<0.0127 0.707 -0.0127 -0.707>}
 spine_01 {parent=pelvis, loc=<6.1208 -0.9328 0>, rot=<0 0 0.034 -0.9994>}
 spine_02 {parent=spine_01, loc=<6.5996 -0.2095 0>, rot=<0 0 0.0238 -0.9997>}
 spine_03 {parent=spine_02, loc=<6.4914 -0.1237 0>, rot=<0 0 0.0149 -0.9999>}
 spine_04 {parent=spine_03, loc=<6.9598 0.0794 0>, rot=<0 0 -0.0132 -0.9999>}
 spine_05 {parent=spine_04, loc=<6.9987 0 0>, rot=<0 0 -0.0015 -1>}
 neck_01 {parent=spine_05, loc=<11.3138 0.1529 0>, rot=<0 0 -0.2003 -0.9797>}
 neck_02 {parent=neck_01, loc=<4.5627 0 0>, rot=<0 0 0.0504 -0.9987>}
 head {parent=neck_02, loc=<5.3914 0 0>, rot=<0 0 0.115 -0.9934>}
 dyn_earring_r {parent=head, loc=<0.2363 -0.0572 6.4098>, rot=<-0.6471 -0.1659 0.7415 0.0625>}
 dyn_earring_l {parent=head, loc=<0.1246 0.8495 -6.6102>, rot=<0.0024 0.361 0.9312 -0.0503>}
 dyn_earring_l1 {parent=dyn_earring_l, loc=<1.881 0 0>, rot=<0 0 0 1>}
 dyn_hairBack_01_Base_l {parent=head, loc=<1.9286 -4.2875 -5.7011>, rot=<0.114 -0.9934 0.0009 -0.0086>}
 dyn_hairBack_01_l {parent=dyn_hairBack_01_Base_l, loc=<0 0 0>, rot=<0 0 0 1>}
 dyn_hairBack_02_l {parent=dyn_hairBack_01_l, loc=<7.8816 0 0>, rot=<-0.0277 0.1434 0.1395 -0.9794>}
 dyn_hairBack_03_l {parent=dyn_hairBack_02_l, loc=<6.0784 0 0>, rot=<0.0257 -0.0722 0.0098 -0.997>}
 dyn_hairBack_01_Base_r {parent=head, loc=<1.4944 -4.7128 4.9508>, rot=<0.1056 -0.9896 -0.0096 0.0974>}
 dyn_hairBack_01_r {parent=dyn_hairBack_01_Base_r, loc=<0 0 0>, rot=<0 0 0 1>}
 dyn_hairBack_02_r {parent=dyn_hairBack_01_r, loc=<8.8354 0 0>, rot=<-0.0133 0.0444 0.3355 -0.9409>}
 dyn_hairBack_03_r {parent=dyn_hairBack_02_r, loc=<7.3453 0 0>, rot=<-0.0128 0.033 -0.2596 -0.9651>}
 dyn_hairTemple_01_Base_l {parent=head, loc=<10.4073 9.2752 -3.2357>, rot=<-0.0087 0.5735 0.8191 0.0014>}
 dyn_hairTemple_01_l {parent=dyn_hairTemple_01_Base_l, loc=<0 0 0>, rot=<0 0 0.2173 -0.9761>}
 dyn_hairTemple_01_l1 {parent=dyn_hairTemple_01_l, loc=<6 0 -0.4607>, rot=<0 0 -0.1969 -0.9804>}
 dyn_hairTemple_01_l2 {parent=dyn_hairTemple_01_l1, loc=<4.8993 0.438 -1.0036>, rot=<-0.0381 -0.1762 -0.2177 -0.9592>}
 dyn_hairTempre_01_Base_r {parent=head, loc=<10.4075 9.2752 3.2357>, rot=<-0.5735 -0.0087 0.0014 -0.8191>}
 dyn_hairTempre_01_r {parent=dyn_hairTempre_01_Base_r, loc=<0 0 0>, rot=<0 0 0.2173 -0.9761>}
 dyn_hairTempre_01_r1 {parent=dyn_hairTempre_01_r, loc=<-5.9999 0 0.4607>, rot=<0 0 -0.1969 -0.9804>}
 dyn_hairTempre_01_r2 {parent=dyn_hairTempre_01_r1, loc=<-4.8991 -0.438 1.0036>, rot=<-0.0381 -0.1762 -0.2177 -0.9592>}
 dyn_hairBack_01_c_BASE {parent=head, loc=<0.591 -5.5907 -0.0581>, rot=<0.1728 -0.9849 0.0006 -0.0039>}
 dyn_hairBack_01_c {parent=dyn_hairBack_01_c_BASE, loc=<0 0 0>, rot=<0 0 0 1>}
 dyn_hairBack_02_c {parent=dyn_hairBack_01_c, loc=<7.7812 0 0>, rot=<-0.0024 0.0145 0.1407 -0.9899>}
 dyn_hairBack_03_c {parent=dyn_hairBack_02_c, loc=<7.032 0 0>, rot=<-0.031 0.1122 -0.0354 -0.9926>}
 clavicle_l {parent=spine_05, loc=<5.4831 -0.1053 -3.2946>, rot=<0.0465 -0.7289 -0.009 -0.6829>}
 upperarm_l {parent=clavicle_l, loc=<9.8658 -1.2136 -0.4105>, rot=<0.0208 -0.486 -0.0242 -0.8734>}
 clavicle_r {parent=spine_05, loc=<5.4831 -0.1053 3.2946>, rot=<0.7289 0.0465 -0.6829 0.009>}
 upperarm_r {parent=clavicle_r, loc=<-9.8658 1.2136 0.4104>, rot=<0.0208 -0.486 -0.0242 -0.8734>}
]
variables [
 Is Rifle Targeting Alpha {type=double, default=0}
 IsNotSkydiving_Moving2D_Crouching {type=double, default=0}
 crouching {type=bool, default=False}
]
