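# Physics asset for the F_MED_Rockstar female player body, covering secondary motion of its hair and left earring.
# Capsules on upper spine, clavicles, neck and head form the body shapes, linked by constraints with free swing and twist.
# Three back-hair and two temple chains hang from head through free-rotating base bodies, with angle-limited, low-twist joints along each chain.
# The two-body earring chain joins head directly with similar limits.
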PhysicsAsset /BRCosmetics/Characters/Player/Female/Medium/Bodies/F_MED_Rockstar/Meshes/Parts/F_MED_Rockstar_Physics
bodies {
 clavicle_l [
  {
   shapeType=Capsule
   length=8.9846
   radius=5.5904
   shapeName=clavicle_l_capsule
   center=<5.6997 -0.1568 0.0156>
   rotation=<rot 80 -0.0003 -0.0003>
  }
  {
   shapeType=Capsule
   length=11.0535
   radius=5.9461
   shapeName=clavicle_l_capsule
   center=<5.5499 8.9226 0.1363>
   rotation=<rot -80.4501 179.9973 -179.9973>
  }
 ]
 clavicle_r [{
   shapeType=Capsule
   length=9.1734
   radius=5.5904
   shapeName=clavicle_r_capsule
   center=<-5.8083 0.0288 0.0051>
   rotation=<rot 80 -0.0003 -0.0004>
  }]
 dyn_earring_l [{shapeType=Capsule, length=1, radius=0.5, center=<1 0 0>, rotation=<rot 90 0 0>}]
 dyn_earring_l1 [{shapeType=Capsule, length=0.505, radius=1, center=<1 0 0>, rotation=<rot 90 0 0>}]
 dyn_hairBack_01_Base_l [{shapeType=Capsule, length=1, radius=1}]
 dyn_hairBack_01_Base_r [{shapeType=Capsule, length=1, radius=1, center=<1.1322 0 0>}]
 dyn_hairBack_01_c [{shapeType=Capsule, length=3.3555, radius=2.4187, center=<5.6863 -0.7908 0>, rotation=<rot 90 -90 -90.0002>}]
 dyn_hairBack_01_c_BASE [{shapeType=Capsule, length=0.505, radius=0.505, shapeName=dyn_hairBack_01_c_BASE_capsule, rotation=<rot 0 -90 0>}]
 dyn_hairBack_01_l [{shapeType=Capsule, length=6.7872, radius=2.3108, center=<4.6636 -0.0001 0>, rotation=<rot 90 0 0>}]
 dyn_hairBack_01_r [{shapeType=Capsule, length=5.9493, radius=1.9533, center=<4.4017 0 0>, rotation=<rot -90 -179.9802 179.9805>}]
 dyn_hairBack_02_c [{shapeType=Capsule, length=2.7965, radius=1.8649, center=<2.5308 0 0>, rotation=<rot 90 0 0>}]
 dyn_hairBack_02_l [{shapeType=Capsule, length=1.5553, radius=2.5729, center=<2.5849 0 0>, rotation=<rot 90 0 0>}]
 dyn_hairBack_02_r [{shapeType=Capsule, length=1.3973, radius=2.6252, center=<3.424 -0.0258 0>, rotation=<rot 90 0 0>}]
 dyn_hairBack_03_c [{shapeType=Capsule, length=0.952, radius=2.2444, center=<0.905 0 0>, rotation=<rot 90 0 0>}]
 dyn_hairBack_03_l [{shapeType=Capsule, length=0.8275, radius=2.7223, center=<1.3953 0 0>, rotation=<rot 90 0 0>}]
 dyn_hairBack_03_r [{shapeType=Capsule, length=0.436, radius=2.5471, center=<1.5777 0.0026 0>, rotation=<rot 90 0 0>}]
 dyn_hairTemple_01_Base_l [{shapeType=Capsule, length=0.5, radius=0.5, rotation=<rot 0 -90 0>}]
 dyn_hairTemple_01_l [{shapeType=Capsule, length=3, radius=0.5, center=<1 0 -0.5306>, rotation=<rot 0 -93.3544 -87.9137>}]
 dyn_hairTemple_01_l1 [{shapeType=Capsule, length=2, radius=0.5, center=<0.2 0 0>, rotation=<rot 90 0 0>}]
 dyn_hairTemple_01_l2 [{shapeType=Capsule, length=2, radius=0.5, center=<0.2 0 0>, rotation=<rot 90 0 0>}]
 dyn_hairTempre_01_Base_r [{shapeType=Capsule, length=0.5, radius=0.5, rotation=<rot 0 -90 0>}]
 dyn_hairTempre_01_r [{shapeType=Capsule, length=3, radius=0.5, center=<-1 0 0.5347>, rotation=<rot 0 -92.3261 -88.049>}]
 dyn_hairTempre_01_r1 [{shapeType=Capsule, length=2, radius=0.5, center=<-0.2 0 0>, rotation=<rot 90 0 0>}]
 dyn_hairTempre_01_r2 [{shapeType=Capsule, length=2, radius=0.5, center=<-0.2 0 0>, rotation=<rot 90 0 0>}]
 head [
  {shapeType=Capsule, length=5, radius=7, center=<2.4021 3.1486 0>, rotation=<rot 90 0 0>}
  {shapeType=Capsule, length=2, radius=2, shapeName=head_capsule, center=<0.9639 8.29 0>, rotation=<rot 90 180 -149.9998>}
 ]
 neck_01 [{shapeType=Capsule, length=25.7589, radius=3.4289, shapeName=neck_01_capsule, center=<0.2818 1.6844 0>, rotation=<rot 0 -90 0>}]
 neck_02 [{shapeType=Capsule, length=20.8724, radius=2.3973, shapeName=neck_02_capsule, center=<0.1956 1.6444 0>}]
 spine_05 [{shapeType=Capsule, length=11.9086, radius=7.8758, shapeName=spine_05_capsule}]
}
constraints [
 {
  bone1Name=dyn_hairBack_01_Base_l
  bone2Name=head
  swing1LimitDegrees=45
  swing1Motion=Free
  swing2LimitDegrees=45
  swing2Motion=Free
  twistLimitDegrees=45
  twistMotion=Free
 }
 {
  bone1Name=dyn_hairBack_01_Base_r
  bone2Name=head
  swing1LimitDegrees=45
  swing1Motion=Free
  swing2LimitDegrees=45
  swing2Motion=Free
  twistLimitDegrees=45
  twistMotion=Free
 }
 {bone1Name=dyn_earring_l, bone2Name=head, swing1LimitDegrees=10, swing2LimitDegrees=10, twistLimitDegrees=1}
 {bone1Name=dyn_earring_l1, bone2Name=dyn_earring_l, swing1LimitDegrees=20, swing2LimitDegrees=30, twistLimitDegrees=5}
 {
  bone1Name=head
  bone2Name=neck_02
  swing1LimitDegrees=45
  swing1Motion=Free
  swing2LimitDegrees=45
  swing2Motion=Free
  twistLimitDegrees=45
  twistMotion=Free
 }
 {
  bone1Name=clavicle_l
  bone2Name=spine_05
  swing1LimitDegrees=45
  swing1Motion=Free
  swing2LimitDegrees=45
  swing2Motion=Free
  twistLimitDegrees=45
  twistMotion=Free
 }
 {
  bone1Name=clavicle_r
  bone2Name=spine_05
  swing1LimitDegrees=45
  swing1Motion=Free
  swing2LimitDegrees=45
  swing2Motion=Free
  twistLimitDegrees=45
  twistMotion=Free
 }
 {bone1Name=dyn_hairTempre_01_r1, bone2Name=dyn_hairTempre_01_r, swing1LimitDegrees=40, swing2LimitDegrees=20, twistLimitDegrees=1}
 {bone1Name=dyn_hairTempre_01_r2, bone2Name=dyn_hairTempre_01_r1, swing1LimitDegrees=20, swing2LimitDegrees=20, twistLimitDegrees=1}
 {bone1Name=dyn_hairTemple_01_l1, bone2Name=dyn_hairTemple_01_l, swing1LimitDegrees=40, swing2LimitDegrees=20, twistLimitDegrees=1}
 {bone1Name=dyn_hairTemple_01_l2, bone2Name=dyn_hairTemple_01_l1, swing1LimitDegrees=20, swing2LimitDegrees=20, twistLimitDegrees=1}
 {bone1Name=dyn_hairBack_01_l, bone2Name=dyn_hairBack_01_Base_l, swing1LimitDegrees=30, swing2LimitDegrees=10, twistLimitDegrees=1}
 {bone1Name=dyn_hairBack_02_l, bone2Name=dyn_hairBack_01_l, swing1LimitDegrees=40, swing2LimitDegrees=15, twistLimitDegrees=10}
 {bone1Name=dyn_hairBack_03_l, bone2Name=dyn_hairBack_02_l, swing1LimitDegrees=35, swing2LimitDegrees=8, twistLimitDegrees=5}
 {bone1Name=dyn_hairBack_01_r, bone2Name=dyn_hairBack_01_Base_r, swing1LimitDegrees=45, swing2LimitDegrees=5, twistLimitDegrees=2}
 {bone1Name=dyn_hairBack_01_c, bone2Name=dyn_hairBack_01_c_BASE, swing1LimitDegrees=25, swing2LimitDegrees=8, twistLimitDegrees=10}
 {bone1Name=dyn_hairBack_02_c, bone2Name=dyn_hairBack_01_c, swing1LimitDegrees=35, swing2LimitDegrees=10, twistLimitDegrees=10}
 {bone1Name=dyn_hairBack_03_c, bone2Name=dyn_hairBack_02_c, swing1LimitDegrees=35, swing2LimitDegrees=7, twistLimitDegrees=5}
 {
  bone1Name=dyn_hairTemple_01_Base_l
  bone2Name=head
  swing1LimitDegrees=45
  swing1Motion=Free
  swing2LimitDegrees=45
  swing2Motion=Free
  twistLimitDegrees=45
  twistMotion=Free
 }
 {
  bone1Name=dyn_hairTempre_01_Base_r
  bone2Name=head
  swing1LimitDegrees=45
  swing1Motion=Free
  swing2LimitDegrees=45
  swing2Motion=Free
  twistLimitDegrees=45
  twistMotion=Free
 }
 {bone1Name=dyn_hairTemple_01_l, bone2Name=dyn_hairTemple_01_Base_l, swing1LimitDegrees=30, swing2LimitDegrees=40, twistLimitDegrees=1}
 {bone1Name=dyn_hairTempre_01_r, bone2Name=dyn_hairTempre_01_Base_r, swing1LimitDegrees=30, swing2LimitDegrees=40, twistLimitDegrees=1}
 {bone1Name=dyn_hairBack_02_r, bone2Name=dyn_hairBack_01_r, swing1LimitDegrees=40, swing2LimitDegrees=5, twistLimitDegrees=3}
 {bone1Name=dyn_hairBack_03_r, bone2Name=dyn_hairBack_02_r, swing1LimitDegrees=35, swing2LimitDegrees=2, twistLimitDegrees=10}
 {
  bone1Name=dyn_hairBack_01_c_BASE
  bone2Name=head
  swing1LimitDegrees=45
  swing1Motion=Free
  swing2LimitDegrees=45
  swing2Motion=Free
  twistLimitDegrees=45
  twistMotion=Free
 }
 {
  bone1Name=neck_01
  bone2Name=spine_05
  swing1LimitDegrees=45
  swing1Motion=Free
  swing2LimitDegrees=45
  swing2Motion=Free
  twistLimitDegrees=45
  twistMotion=Free
 }
 {
  bone1Name=neck_02
  bone2Name=neck_01
  swing1LimitDegrees=45
  swing1Motion=Free
  swing2LimitDegrees=45
  swing2Motion=Free
  twistLimitDegrees=45
  twistMotion=Free
 }
]
