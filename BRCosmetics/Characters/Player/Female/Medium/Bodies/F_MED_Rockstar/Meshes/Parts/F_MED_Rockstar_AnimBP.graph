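# Rockstar female cosmetic part AnimBP that poses the character's back hair.
# It copies the parent pose, layers body-mesh upper arms, runs neck-relative RigidBody physics with gravity flipped upward in skydive dives, then nudges two right-side hair bones when HeadUp.
# The update graph derives that alpha from head-versus-spine twist, faded out during skydiving and emotes.
# Control Rig and other hair-tweak chains sit unconnected to the output pose.
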AnimBlueprint /BRCosmetics/Characters/Player/Female/Medium/Bodies/F_MED_Rockstar/Meshes/Parts/F_MED_Rockstar_AnimBP
target_skeleton=&/F_MED_Rockstar_Skeleton
graphs [
 AnimGraph {nodes [
   Control Rig!(
    Source
    Alpha
    crouching:bool=Get bIsCrouching()
    Is Rifle Targeting Alpha:double=Get Is Rifle Targeting Alpha() -> (Is Rifle Targeting Alpha:float)
    IsNotSkydiving_Moving2D_Crouching:double=Get IsNotSkydiving_Moving2D_Crouching() -> (IsNotSkydiving_Moving2D_Crouching:double)
   ) -> (Pose) {
    alphaBoolBlend {blendOption=Linear}
    controlRigAssetReference {blueprintRigClass=&/F_MED_Rockstar_CtrlRig.F_MED_Rockstar_CtrlRig_C}
   }
   Input Pose()
   n1 = "Transform (Modify) Bone - Bone: dyn_hairBack_01_c"(Alpha=1, Translation="0,0,0", Rotation="0,0,0", Scale="1,1,1") {
    rotationMode=BMM_Additive
    rotationSpace=BCS_BoneSpace
    translationMode=BMM_Additive
    translationSpace=BCS_BoneSpace
    alphaBoolBlend {blendOption=Linear}
    boneToModify {boneName=dyn_hairBack_01_c}
   }
   "Transform (Modify) Bone - Bone: dyn_hairBack_01_Base_l"(ComponentPose=@n1, Alpha=1, Translation="0,0,0", Rotation="0,0,0", Scale="1,1,1") {
    rotationMode=BMM_Additive
    rotationSpace=BCS_BoneSpace
    translationMode=BMM_Additive
    translationSpace=BCS_BoneSpace
    alphaBoolBlend {blendOption=Linear}
    boneToModify {boneName=dyn_hairBack_01_Base_l}
   }
   n2 = Get FronEnd_Not_SkyDiveDiveMode_Not_Emote_Hair_Tweak() -> (FronEnd_Not_SkyDiveDiveMode_Not_Emote_Hair_Tweak:double)
   n3 = "Transform (Modify) Bone - Bone: dyn_hairBack_02_l"(Alpha=@n2, Translation="0.643,-0.101,0.120", Rotation="-5.191,13.573,-1.744", Scale="1,1,1") {
    rotationMode=BMM_Additive
    rotationSpace=BCS_BoneSpace
    translationMode=BMM_Additive
    translationSpace=BCS_BoneSpace
    alphaBoolBlend {blendOption=Linear}
    boneToModify {boneName=dyn_hairBack_02_l}
    rotation=<rot -5.191 13.5731 -1.744>
    translation=<0.643 -0.101 0.12>
   }
   n4 = "Transform (Modify) Bone - Bone: dyn_hairBack_01_c"(ComponentPose=@n3, Alpha=@n2, Translation="1.476,3.743,2.853", Rotation="-0.529,-6.259,1.603", Scale="1,1,1") {
    rotationMode=BMM_Additive
    rotationSpace=BCS_BoneSpace
    translationMode=BMM_Additive
    translationSpace=BCS_BoneSpace
    alphaBoolBlend {blendOption=Linear}
    boneToModify {boneName=dyn_hairBack_01_c}
    rotation=<rot -0.529 -6.2589 1.603>
    translation=<1.476 3.743 2.853>
   }
   "Transform (Modify) Bone - Bone: dyn_hairBack_01_r"(ComponentPose=@n4, Alpha=@n2, Translation="-0.366,1.192,-0.773", Rotation="1.465,9.491,4.760", Scale="1,1,1") {
    rotationMode=BMM_Additive
    rotationSpace=BCS_BoneSpace
    translationMode=BMM_Additive
    translationSpace=BCS_BoneSpace
    alphaBoolBlend {blendOption=Linear}
    boneToModify {boneName=dyn_hairBack_01_r}
    rotation=<rot 1.465 9.491 4.76>
    translation=<-0.366 1.192 -0.773>
   }
   n5 = Get SkyDiveDiveMode_Hair_Tweak() -> (SkyDiveDiveMode_Hair_Tweak:double)
   n6 = "Transform (Modify) Bone - Bone: dyn_hairBack_02_l"(Alpha=@n5, Translation="0,0,0", Rotation="0,0,0", Scale="1,1,1") {
    rotationMode=BMM_Additive
    rotationSpace=BCS_BoneSpace
    translationMode=BMM_Additive
    translationSpace=BCS_BoneSpace
    alphaBoolBlend {blendOption=Linear}
    boneToModify {boneName=dyn_hairBack_02_l}
   }
   n7 = "Transform (Modify) Bone - Bone: dyn_hairBack_01_c"(ComponentPose=@n6, Alpha=@n5, Translation="0,0,0", Rotation="0,0,0", Scale="1,1,1") {
    rotationMode=BMM_Additive
    rotationSpace=BCS_BoneSpace
    translationMode=BMM_Additive
    translationSpace=BCS_BoneSpace
    alphaBoolBlend {blendOption=Linear}
    boneToModify {boneName=dyn_hairBack_01_c}
   }
   "Transform (Modify) Bone - Bone: dyn_hairBack_01_r"(ComponentPose=@n7, Alpha=@n5, Translation="0,0,0", Rotation="0,0,0", Scale="1,1,1") {
    rotationMode=BMM_Additive
    rotationSpace=BCS_BoneSpace
    translationMode=BMM_Additive
    translationSpace=BCS_BoneSpace
    alphaBoolBlend {blendOption=Linear}
    boneToModify {boneName=dyn_hairBack_01_r}
   }
   n8 = Get HeadUp_and_ IsRifle_and_IsCrouching() -> (HeadUp_and_ IsRifle_and_IsCrouching:double)
   n9 = "Transform (Modify) Bone - Bone: dyn_hairBack_02_r"(Alpha=@n8, Translation="0,1,0", Rotation="0,15,0", Scale="1,1,1") {
    rotationMode=BMM_Additive
    rotationSpace=BCS_BoneSpace
    translationMode=BMM_Additive
    translationSpace=BCS_BoneSpace
    alphaBoolBlend {blendOption=Linear}
    boneToModify {boneName=dyn_hairBack_02_r}
    rotation {yaw=15}
    translation {y=1}
   }
   n10 = "Transform (Modify) Bone - Bone: dyn_hairBack_01_Base_r"(ComponentPose=@n9, Alpha=@n8, Translation="-1.121,0.143,0.608", Rotation="0,0,0", Scale="1,1,1") {
    rotationMode=BMM_Additive
    rotationSpace=BCS_BoneSpace
    translationMode=BMM_Additive
    translationSpace=BCS_BoneSpace
    alphaBoolBlend {blendOption=Linear}
    boneToModify {boneName=dyn_hairBack_01_Base_r}
    translation=<-1.121 0.143 0.608>
   }
   n11 = Get HeadDown_and_IsRifle_and_Crouch() -> (HeadDown_and_IsRifle_and_Crouch:double)
   n12 = "Transform (Modify) Bone - Bone: dyn_hairBack_02_r"(ComponentPose=@n10, Alpha=@n11, Translation="-0.214,2,0.429", Rotation="0,35,0", Scale="1,1,1") {
    rotationMode=BMM_Additive
    rotationSpace=BCS_BoneSpace
    translationMode=BMM_Additive
    translationSpace=BCS_BoneSpace
    alphaBoolBlend {blendOption=Linear}
    boneToModify {boneName=dyn_hairBack_02_r}
    rotation {yaw=35}
    translation=<-0.214 2 0.429>
   }
   n13 = "Transform (Modify) Bone - Bone: dyn_hairBack_01_Base_r"(ComponentPose=@n12, Alpha=@n11, Translation="0,0,0", Rotation="0,-5,0", Scale="1,1,1") {
    rotationMode=BMM_Additive
    rotationSpace=BCS_BoneSpace
    translationMode=BMM_Additive
    translationSpace=BCS_BoneSpace
    alphaBoolBlend {blendOption=Linear}
    boneToModify {boneName=dyn_hairBack_01_Base_r}
    rotation {yaw=-5}
   }
   n14 = Get NotHeadUp_and_NotHeadDown_and_IsNotSkydiving_Moving2D_Crouching() -> (NotHeadUp_and_NotHeadDown_and_IsNotSkydiving_Moving2D_Crouching:double)
   n15 = "Transform (Modify) Bone - Bone: dyn_hairBack_02_r"(ComponentPose=@n13, Alpha=@n14, Translation="0,0,0", Rotation="0,-10,0", Scale="1,1,1") {
    rotationMode=BMM_Additive
    rotationSpace=BCS_BoneSpace
    translationMode=BMM_Additive
    translationSpace=BCS_BoneSpace
    alphaBoolBlend {blendOption=Linear}
    boneToModify {boneName=dyn_hairBack_02_r}
    rotation {yaw=-10}
   }
   n16 = "Transform (Modify) Bone - Bone: dyn_hairBack_01_Base_r"(ComponentPose=@n15, Alpha=@n14, Translation="0,0,0", Rotation="0,0,0", Scale="1,1,1") {
    rotationMode=BMM_Additive
    rotationSpace=BCS_BoneSpace
    translationMode=BMM_Additive
    translationSpace=BCS_BoneSpace
    alphaBoolBlend {blendOption=Linear}
    boneToModify {boneName=dyn_hairBack_01_Base_r}
   }
   n17 = "Transform (Modify) Bone - Bone: dyn_hairBack_02_r"(ComponentPose=@n16, Alpha=Get Head turn() -> (Head turn:double), Translation="-1,0,2", Rotation="0,0,0", Scale="1,1,1") {
    rotationMode=BMM_Additive
    rotationSpace=BCS_BoneSpace
    translationMode=BMM_Additive
    translationSpace=BCS_BoneSpace
    alphaBoolBlend {blendInTime=1, blendOption=Linear, blendOutTime=0.2}
    boneToModify {boneName=dyn_hairBack_02_r}
    translation {x=-1, z=2}
   }
   n18 = "Transform (Modify) Bone - Bone: dyn_hairBack_02_r"(ComponentPose=@n17, Alpha=Get Hair_Tweak() -> (Hair_Tweak:double), Translation="1.326,-9.776,-1.908", Rotation="0,0,0", Scale="1,1,1") {
    rotationMode=BMM_Additive
    rotationSpace=BCS_BoneSpace
    translationMode=BMM_Additive
    translationSpace=BCS_BoneSpace
    alphaBoolBlend {blendInTime=1, blendOption=Linear, blendOutTime=0.2}
    boneToModify {boneName=dyn_hairBack_02_r}
    translation=<1.326 -9.776 -1.908>
   }
   "Transform (Modify) Bone - Bone: dyn_hairBack_01_r"(ComponentPose=@n18, Alpha=Get Hair_Tweak(), Translation="0,1.016,-0.076", Rotation="0,0,0", Scale="1,1,1") {
    rotationMode=BMM_Additive
    rotationSpace=BCS_BoneSpace
    translationMode=BMM_Additive
    translationSpace=BCS_BoneSpace
    alphaBoolBlend {blendInTime=1, blendOption=Linear, blendOutTime=0.2}
    boneToModify {boneName=dyn_hairBack_01_r}
    translation {y=1.016, z=-0.076}
   }
   n19 = "Transform (Modify) Bone - Bone: dyn_hairBack_02_c"(Alpha=1, Translation="0,0,0", Rotation="0,0,0", Scale="1,1,1") {
    rotationMode=BMM_Additive
    rotationSpace=BCS_BoneSpace
    translationMode=BMM_Additive
    translationSpace=BCS_BoneSpace
    alphaBoolBlend {blendOption=Linear}
    boneToModify {boneName=dyn_hairBack_02_c}
   }
   "Transform (Modify) Bone - Bone: dyn_hairBack_02_l"(ComponentPose=@n19, Alpha=1, Translation="0,0,0", Rotation="0,0,0", Scale="1,1,1") {
    rotationMode=BMM_Additive
    rotationSpace=BCS_BoneSpace
    translationMode=BMM_Additive
    translationSpace=BCS_BoneSpace
    alphaBoolBlend {blendOption=Linear}
    boneToModify {boneName=dyn_hairBack_02_l}
   }
   n20 = Copy Pose From Mesh() {bCopyCurves=true, bUseAttachedParent=true}
   n21 = Copy Pose From Mesh(SourceMeshComponent=Get body Mesh() -> (body Mesh:Skeletal Mesh Component Object Reference))
   n22 = RigidBody!(
    ComponentPose=Local To Component(LocalPose=Layered blend per bone(BasePose=@n20, BlendPoses_0=@n21) {layerSetup [{branchFilters [{boneName=upperarm_l}, {boneName=upperarm_r}]}]})
    Alpha
    OverrideWorldGravity:Vector=Get Override_World_Gravity() -> (Override_World_Gravity:Vector)
    ExternalForce
   ) -> (Pose) {
    bOverrideWorldGravity=true
    overridePhysicsAsset=&/F_MED_Rockstar_Physics
    simulationSpace=BaseBoneSpace
    alphaBoolBlend {blendOption=Linear}
    baseBoneRef {boneName=neck_02}
    componentAppliedLinearAccClamp=<1200 1200 1200>
    componentLinearAccScale=<0.1 0.8 0.8>
    componentLinearVelScale=<0.1 0.8 0.1>
    overrideWorldGravity {z=-980}
   }
   n23 = Get HeadUp_NotSkyDive_NotEmote() -> (HeadUp_NotSkyDive_NotEmote:double)
   n24 = "Transform (Modify) Bone - Bone: dyn_hairBack_01_Base_r"(ComponentPose=@n22, Alpha=@n23, Translation="-1,0.5,0.9", Rotation="0,0,0", Scale="1,1,1") {
    rotationMode=BMM_Additive
    rotationSpace=BCS_BoneSpace
    translationMode=BMM_Additive
    translationSpace=BCS_BoneSpace
    alphaBoolBlend {blendOption=Linear}
    boneToModify {boneName=dyn_hairBack_01_Base_r}
    translation=<-1 0.5 0.9>
   }
   n25 = "Transform (Modify) Bone - Bone: dyn_hairBack_02_r"(ComponentPose=@n24, Alpha=@n23, Translation="-2.5,0.563,1", Rotation="0,0,-1", Scale="1,1,1") {
    rotationMode=BMM_Additive
    rotationSpace=BCS_BoneSpace
    translationMode=BMM_Additive
    translationSpace=BCS_BoneSpace
    alphaBoolBlend {blendOption=Linear}
    boneToModify {boneName=dyn_hairBack_02_r}
    rotation {roll=-1}
    translation=<-2.5 0.563 1>
   }
   Output Pose(Result=Component To Local(ComponentPose=@n25)) {layerGroup=None}
  ]}
 EventGraph {nodes [
   n1 = Event BlueprintUpdateAnimation()
   > Set DeltaTimeX(DeltaTimeX:double=@n1.DeltaTimeX) -> (Output_Get:double)
   > n2 = Sequence!(execute) -> (then_0, then_1, then_2:Exec, then_3:Exec, then_4:Exec, then_5:Exec, then_6:Exec, then_7:Exec, then_8:Exec)
   > Set HeadUp(HeadUp:double=GetAlphaFromReferenceTransformDelta(
     SkeletalMeshComponent=Get Skeleton() -> (Skeleton:Skeletal Mesh Component Object Reference)
     BoneName=head
     BaseName=spine_05
     DeltaComponentType=QuaternionTwist
     InputRange=(X=5,Y=45)
     TwistAxis="0,0,1"
    )) -> (Output_Get:double)
   > Set HeadDown(HeadDown:double=GetAlphaFromReferenceTransformDelta(
     SkeletalMeshComponent=Get Skeleton()
     BoneName=head
     BaseName=spine_05
     DeltaComponentType=QuaternionTwist
     InputRange=(X=4,Y=-5)
     TwistAxis="0,0,1"
    )) -> (Output_Get:double)
   n3 = Cast To FortPlayerAnimInstance(execute=@n2.then_4, Object=GetAnimInstance(self=GetOwnerSkeletalMeshComponent()))
   > Set isPistol(isPistol=false)
   > Set isRifle(isRifle:bool=false) -> (Output_Get:bool)
   > n4 = Switch on EFortWeaponCoreAnimation!(execute, Selection:EFortWeaponCoreAnimation Enum=Get WeaponCoreAnim(self=@n3)) -> (
    Melee:Exec
    Pistol:Exec
    Shotgun:Exec
    PaperBlueprint:Exec
    Rifle:Exec
    MeleeOneHand:Exec
    MachinePistol:Exec
    RocketLauncher:Exec
    GrenadeLauncher:Exec
    GoingCommando:Exec
    AssaultRifle:Exec
    TacticalShotgun:Exec
    SniperRifle:Exec
    TrapPlacement:Exec
    ShoulderLauncher:Exec
    AbilityDecoTool:Exec
    Crossbow:Exec
    C4:Exec
    RemoteControl:Exec
    DualWield:Exec
    AR_BullPup:Exec
    AR_ForwardGrip:Exec
    MedPackPaddles:Exec
    SMG_P90:Exec
    AR_DrumGun:Exec
    Consumable_Small:Exec
    Consumable_Large:Exec
    Balloon:Exec
    MountedTurret:Exec
    CreativeTool:Exec
    ExplosiveBow:Exec
    AshtonIndigo:Exec
    AshtonChicago:Exec
    MeleeDualWield:Exec
    Unarmed:Exec
   )
   Set isRifle(execute=[@n4.AssaultRifle, @n4.Rifle, @n4.RocketLauncher, @n4.Shotgun, @n4.SniperRifle, @n4.TacticalShotgun], isRifle=true)
   Set isPistol(execute=@n4.Pistol, isPistol=true)
   n5 = Get isRifle() -> (isRifle:bool)
   n6 = Get bIsTargeting()
   Set Is Rifle Alpha(execute=@n2.then_5, Is Rifle Alpha:float="To Float (Boolean)"(InBool=AND Boolean(A=@n5, B=@n6))) -> (Output_Get:float)
   n7 = Get bIsTargeting()
   n8 = Get isRifle()
   Set Is Rifle Targeting Alpha(execute=@n2.then_6, Is Rifle Targeting Alpha:float="To Float (Boolean)"(InBool=AND Boolean(A=@n8, B=@n7))) -> (Output_Get:float)
   n9 = Branch(execute=@n2.then_1, Condition=Get bIsJumping())
   > Set Head turn(Head turn:double=0) -> (Output_Get:double)
   Set Head turn(
    execute=@n9.else
    Head turn=GetAlphaFromReferenceTransformDelta(
     SkeletalMeshComponent=Get Skeleton()
     BoneName=head
     BaseName=Spine_05
     DeltaComponentType=QuaternionTwist
     InputRange=(X=30,Y=50)
     TwistAxis="1,0,0"
    )
   )
   n10 = Set Head_roll(
    execute=@n2.then_2
    Head_roll:double=GetAlphaFromReferenceTransformDelta(
     SkeletalMeshComponent=Get Skeleton()
     BoneName=head
     BaseName=Spine_05
     DeltaComponentType=QuaternionTwist
     InputRange=(X=0,Y=15)
     TwistAxis="0,1,0"
    )
   ) -> (Output_Get:double)
   > Set Hair_Tweak(Hair_Tweak:double=float * float(A=@n10, B=Get Is_Crouch() -> (Is_Crouch:double)) {is=title K2Node_PromotableOperator}) -> (Output_Get:double)
   n11 = Get bIsPlayingEmote()
   n12 = Get bIsSkydiveDiveMode()
   n13 = Branch(execute=@n2.then_7, Condition=AND Boolean!(A=Get bIsInFrontEnd(), B=NOT Boolean(A=@n12), C:bool=NOT Boolean(A=@n11)) -> (ReturnValue))
   > Set FronEnd_Not_SkyDiveDiveMode_Not_Emote_Hair_Tweak(FronEnd_Not_SkyDiveDiveMode_Not_Emote_Hair_Tweak:double=FInterpTo_Constant!(
     Current:double=Get FronEnd_Not_SkyDiveDiveMode_Not_Emote_Hair_Tweak()
     Target:double=1
     DeltaTime:double=Get DeltaTimeX() -> (DeltaTimeX:double)
     InterpSpeed:double=2
    ) -> (ReturnValue:double)) -> (Output_Get:double)
   Set FronEnd_Not_SkyDiveDiveMode_Not_Emote_Hair_Tweak(
    execute=@n13.else
    FronEnd_Not_SkyDiveDiveMode_Not_Emote_Hair_Tweak=FInterpTo_Constant(Current=Get FronEnd_Not_SkyDiveDiveMode_Not_Emote_Hair_Tweak(), Target=0, DeltaTime=Get DeltaTimeX(), InterpSpeed=2)
   )
   n14 = Get bIsPlayingEmote()
   n15 = Get bIsSkyDiving()
   n16 = Branch(execute=@n2.then_3, Condition=AND Boolean!(A=NOT Boolean(A=@n14), B=NOT Boolean(A=@n15)) -> (ReturnValue))
   > n17 = Set NotSkyDive_NotEmote(NotSkyDive_NotEmote:double=FInterpTo_Constant(Current=Get NotSkyDive_NotEmote() -> (NotSkyDive_NotEmote:double), Target=1, DeltaTime=Get DeltaTimeX(), InterpSpeed=2)) -> (Output_Get:double)
   Set HeadUp_NotSkyDive_NotEmote(
    execute=[@n18.then, @n17.then]
    HeadUp_NotSkyDive_NotEmote:double=float * float(A=Get HeadUp() -> (HeadUp:double), B=Get NotSkyDive_NotEmote())
   ) -> (Output_Get:double)
   n18 = Set NotSkyDive_NotEmote(
    execute=@n16.else
    NotSkyDive_NotEmote=FInterpTo_Constant(Current=Get NotSkyDive_NotEmote(), Target=0, DeltaTime=Get DeltaTimeX(), InterpSpeed=2)
   )
   n19 = Get bIsInFrontEnd()
   n20 = Branch(execute=@n2.then_8, Condition=AND Boolean(A=Get bIsSkydiveDiveMode(), B=NOT Boolean(A=@n19)))
   > Set Override_World_Gravity(Override_World_Gravity:Vector="0,0,490") -> (Output_Get:Vector)
   Set Override_World_Gravity(execute=@n20.else, Override_World_Gravity="0,0,-980")
   Event BlueprintInitializeAnimation()
   > Set Skeleton(Skeleton:Skeletal Mesh Component Object Reference=GetOwnerSkeletalMeshComponent()) -> (Output_Get:Skeletal Mesh Component Object Reference)
   > Set body Mesh(body Mesh:Skeletal Mesh Component Object Reference=GetOwnerSkeletalMeshComponent()) -> (Output_Get:Skeletal Mesh Component Object Reference)
   Reroute Node()
   n21 = Get bIsSkyDiving()
   > n22 = Branch(Condition=AND Boolean!(A=Get isRifle(), B=Get bIsCrouching(), C:bool=NOT Boolean(A=@n21)) -> (ReturnValue))
   > n23 = Set is_Rifle_and_IsCrouching(is_Rifle_and_IsCrouching:double=FInterpTo_Constant(Current=Get is_Rifle_and_IsCrouching() -> (is_Rifle_and_IsCrouching:double), Target=1, DeltaTime=Get DeltaTimeX(), InterpSpeed=2)) -> (Output_Get:double)
   Set HeadDown_and_IsRifle_and_Crouch(
    execute=[@n24.then, @n23.then]
    HeadDown_and_IsRifle_and_Crouch:double=float * float(A=Get is_Rifle_and_IsCrouching(), B=Get HeadDown() -> (HeadDown:double))
   ) -> (Output_Get:double)
   n24 = Set is_Rifle_and_IsCrouching(
    execute=@n22.else
    is_Rifle_and_IsCrouching=FInterpTo_Constant(Current=Get is_Rifle_and_IsCrouching(), Target=0, DeltaTime=Get DeltaTimeX(), InterpSpeed=2)
   )
   Reroute Node()
   n25 = Get bIsSkyDiving()
   > n26 = Branch(Condition=AND Boolean(A=NOT Boolean(A=@n25), B=Get bIsMoving2D(), C=Get bIsCrouching()))
   > n27 = Set IsNotSkydiving_Moving2D_Crouching(IsNotSkydiving_Moving2D_Crouching:double=FInterpTo_Constant(Current=Get IsNotSkydiving_Moving2D_Crouching(), Target=1, DeltaTime=Get DeltaTimeX(), InterpSpeed=1)) -> (Output_Get:double)
   n28 = Get HeadUp()
   Set NotHeadUp_and_NotHeadDown_and_IsNotSkydiving_Moving2D_Crouching(
    execute=[@n29.then, @n27.then]
    NotHeadUp_and_NotHeadDown_and_IsNotSkydiving_Moving2D_Crouching:double=float * float(A=Get IsNotSkydiving_Moving2D_Crouching(), B=MapRangeClamped(Value=@n28, InRangeA=0.8, InRangeB=1, OutRangeA=1, OutRangeB=0))
   ) -> (Output_Get:double)
   n29 = Set IsNotSkydiving_Moving2D_Crouching(
    execute=@n26.else
    IsNotSkydiving_Moving2D_Crouching=FInterpTo_Constant(Current=Get IsNotSkydiving_Moving2D_Crouching(), Target=0, DeltaTime=Get DeltaTimeX(), InterpSpeed=1)
   )
   Reroute Node()
   n30 = Get bIsMoving2D()
   n31 = Get bIsSkyDiving()
   > n32 = Branch(Condition=AND Boolean!(A=Get isRifle(), B=NOT Boolean(A=@n31), C=Get bIsCrouching(), D:bool=NOT Boolean(A=@n30)) -> (ReturnValue))
   > n33 = Set is_Rifle(is_Rifle:double=FInterpTo_Constant(Current=Get is_Rifle() -> (is_Rifle:double), Target=1, DeltaTime=Get DeltaTimeX(), InterpSpeed=2)) -> (Output_Get:double)
   Set HeadUp_and_ IsRifle_and_IsCrouching(execute=[@n34.then, @n33.then], HeadUp_and_ IsRifle_and_IsCrouching:double=float * float(A=Get is_Rifle(), B=Get HeadUp())) -> (Output_Get:double)
   n34 = Set is_Rifle(execute=@n32.else, is_Rifle=FInterpTo_Constant(Current=Get is_Rifle(), Target=0, DeltaTime=Get DeltaTimeX(), InterpSpeed=2))
   Reroute Node()
   n35 = Get bRigidBodyEnabled()
   > n36 = Branch(Condition=AND Boolean!(A=Get bIsCrouching(), B=NOT Boolean(A=@n35)) -> (ReturnValue))
   > Set Is_Crouch(Is_Crouch:double=FInterpTo_Constant(Current=Get Is_Crouch(), Target=1, DeltaTime=Get DeltaTimeX(), InterpSpeed=2)) -> (Output_Get:double)
   Set Is_Crouch(execute=@n36.else, Is_Crouch=FInterpTo_Constant(Current=Get Is_Crouch(), Target=0, DeltaTime=Get DeltaTimeX(), InterpSpeed=2))
   Reroute Node()
   n37 = Get bIsInFrontEnd()
   > n38 = Branch(Condition=AND Boolean(A=Get bIsSkydiveDiveMode(), B=NOT Boolean(A=@n37)))
   > Set SkyDiveDiveMode_Hair_Tweak(SkyDiveDiveMode_Hair_Tweak:double=FInterpTo_Constant(Current=Get SkyDiveDiveMode_Hair_Tweak(), Target=1, DeltaTime=Get DeltaTimeX(), InterpSpeed=2)) -> (Output_Get:double)
   Set SkyDiveDiveMode_Hair_Tweak(
    execute=@n38.else
    SkyDiveDiveMode_Hair_Tweak=FInterpTo_Constant(Current=Get SkyDiveDiveMode_Hair_Tweak(), Target=0, DeltaTime=Get DeltaTimeX(), InterpSpeed=2)
   )
   NOT Boolean(A=Get bIsMoving2D())
   "float > float"(A=Get PawnSpeed2D(), B=100) {is=title K2Node_CallFunction}
   Set body Mesh(body Mesh=GetOwnerSkeletalMeshComponent())
  ]}
]
variables [
 body Mesh {type=SkeletalMeshComponent}
 CtrlRig_HeadDown_SPR {type=number}
 DeltaTimeX {type=number}
 FronEnd_Not_SkyDiveDiveMode_Not_Emote_Hair_Tweak {type=number}
 Hair_Tweak {type=number}
 Head turn {type=number}
 Head_roll {type=number}
 HeadDown {type=number}
 HeadDown_and_IsRifle_and_Crouch {type=number}
 HeadUp {type=number}
 HeadUp_and_ IsRifle_and_IsCrouching {type=number}
 HeadUp_NotSkyDive_NotEmote {type=number}
 Is Rifle Alpha {type=number}
 Is Rifle Targeting Alpha {type=number}
 Is_Crouch {type=number}
 is_Rifle {type=number}
 is_Rifle_and_IsCrouching {type=number}
 IsNotSkydiving_Moving2D_Crouching {type=number}
 isPistol {type=boolean}
 isRifle {type=boolean}
 NotHeadUp_and_NotHeadDown_and_IsNotSkydiving_Moving2D_Crouching {type=number}
 NotSkyDive_NotEmote {type=number}
 Override_World_Gravity {type=Vector}
 Skeleton {type=SkeletalMeshComponent}
 SkyDiveDiveMode_Hair_Tweak {type=number}
]
